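# Masked parent material for the Esports05 Pro vehicle wheel, with recolorable rim and tire, ending in shared damage and de-rez functions.
# Two shared setup calls turn RGB masks into PBR attributes, blended by vertex color red.
# The tire adds Fresnel-tinted color, panning UVs and a noise normal; the rim's blue mask gets animated noise emissive.
# Reflection-vector noise creates sparkles, and a switchable opacity texture masks the rim for LODs.
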Material /VehicleCosmetics/Wheels/Esports05_Pro/Materials/M_WheelParent_Esports05_Pro
graph {
 properties {
  bAutomaticallySetUsageInEditor=false
  bUseMaterialAttributes=true
  bUsedWithMeshDeformer=true
  bUsedWithSkeletalMesh=true
  blendMode=BLEND_Masked
  pixelDepthOffsetMode=PDOM_Legacy
  usedShadingModels=Default Lit
 }
 nodes [
  n1 = ScalarParameter() {
   defaultValue=0.4
   desc=General Roughness that takes effect in absence of masking.
   group=4_Surface Defaults
   parameterName=Rim_RoughnessBase
   sliderMax=1
  }
  n2 = ScalarParameter() {
   defaultValue=1
   desc=General Metalness that takes effect in absence of masking.
   group=4_Surface Defaults
   parameterName=Rim_MetallicBase
   sliderMax=1
  }
  n3 = ScalarParameter() {
   defaultValue=1
   desc="0==False, 1==True: AO gradient affects Emissive values. Use this for Emissives with depth. Values between 0 && 1 are allowed."
   group=5_Special Effects
   parameterName=Rim_AOTintsEmissive
  }
  n4 = ScalarParameter() {
   desc="0==False, 1==True: Blends AO with Diffuse. Values between 0 && 1 are allowed."
   group=5_Special Effects
   parameterName=Rim_AOTintsDiffuse
  }
  n5 = MaterialFunctionCall(R=@n1, G=@n2, B=@n3, A=@n4) {desc="Vector4 container for Rough, Metallic, AO settings", materialFunction=&AppendMany}
  n6 = TextureSampleParameter2D() {
   group=1_Input Textures
   parameterName=Rim_Normal
   samplerSource=SSM_Wrap_WorldGroupSettings
   samplerType=SAMPLERTYPE_Normal
   texture=/VehicleCosmetics/Textures/T_BlankNormal_B
  }
  n7 = TextureSampleParameter2D() {desc=Diffuse Input Texture - Optional, group=1_Input Textures, parameterName=Rim_DiffuseTexture, texture=/VehicleCosmetics/Textures/T_White}
  n8 = ScalarParameter() {
   desc="0==False, 1==True: Set at 1 if Diffuse overrides color, 0 for no Diffuse. Auto-rounds to 0 or 1."
   group=1_Input Textures
   parameterName=Rim_DiffuseTextureEnable
   sliderMax=1
  }
  n9 = VectorParameter() {
   desc="Albedo/Diffuse color control in absence of RGB maps, or to fill in the gaps. Leave white if using Diffuse Texture instead."
   group=4_Surface Defaults
   parameterName=Rim_ColorBase
   defaultValue=<rgba 1 1 1 1>
  }
  n10 = TextureSampleParameter2D() {
   desc=Mask for Rim - Alpha is AO
   group=1_Input Textures
   parameterName=Rim_MaskRGBA
   samplerType=SAMPLERTYPE_LinearColor
   texture=/VehicleCosmetics/Wheels/Esports05_Pro/Textures/Wheel_Mendoza_Rim_RGB2_pro
   channelNames {a=AO, b=ColorC, g=ColorB, r=ColorA}
  }
  n11 = VectorParameter() {
   desc=Color fill for R value of RGB Mask
   group=2_Rim Properties
   parameterName=Rim_ColorA
   channelNames {a=Unused, b="", g="", r=""}
   defaultValue=<rgba 1 0 0 1>
  }
  n12 = VectorParameter() {
   desc="Emissive, Roughness, Metallic values for ColorA. Emissive has no limit, Roughness && Metallic are clamped between 0 && 1."
   group=2_Rim Properties
   parameterName=Rim_ColorA_PBR
   channelNames {a=Unused, b=Metallic, g=Roughness, r=Emissive}
   defaultValue=<rgba 0 0.4 1 1>
  }
  n13 = VectorParameter() {
   desc=Color fill for G value of RGB Mask
   group=2_Rim Properties
   parameterName=Rim_ColorB
   channelNames {a=Unused, b="", g="", r=""}
   defaultValue=<rgba 0 1 0 1>
  }
  n14 = VectorParameter() {
   desc="Emissive, Roughness, Metallic values for ColorB. Emissive has no limit, Roughness && Metallic are clamped between 0 && 1."
   group=2_Rim Properties
   parameterName=Rim_ColorB_PBR
   channelNames {a=Unused, b=Metallic, g=Roughness, r=Emissive}
   defaultValue=<rgba 0 0.4 1 1>
  }
  n15 = VectorParameter() {
   desc=Color fill for B value of RGB Mask
   group=2_Rim Properties
   parameterName=Rim_ColorC
   channelNames {a=Unused, b="", g="", r=""}
   defaultValue=<rgba 0 0 1 1>
  }
  n16 = VectorParameter() {
   desc="Emissive, Roughness, Metallic values for ColorC. Emissive has no limit, Roughness && Metallic are clamped between 0 && 1."
   group=2_Rim Properties
   parameterName=Rim_ColorC_PBR
   channelNames {a=Unused, b=Metallic, g=Roughness, r=Emissive}
   defaultValue=<rgba 0 0.4 1 1>
  }
  n17 = MaterialFunctionCall!(
   Base_Settings=@n5.RGBA
   Normal_Texture=@n6.RGB
   Diffuse_Texture=@n7.RGB
   DiffuseTextureEnable=@n8
   ColorBase=@n9.RGB
   MaskRGBA=@n10.RGBA
   ColorA_Color=@n11
   ColorA_PBR_Values=@n12
   ColorB_Color=@n13
   ColorB_PBR_Values=@n14
   ColorC_Color=@n15
   ColorC_PBR_Values=@n16
  ) {materialFunction=/VehicleCosmetics/Wheels/_SharedMaterials/MF_Wheel_Parent_Simple_InputSetup}
  n18 = ScalarParameter() {
   defaultValue=0.7
   desc=General Roughness that takes effect in absence of masking.
   group=4_Surface Defaults
   parameterName=Tire_RoughnessBase
   sliderMax=1
  }
  n19 = ScalarParameter() {desc=General Metalness that takes effect in absence of masking., group=4_Surface Defaults, parameterName=Tire_MetallicBase, sliderMax=1}
  n20 = ScalarParameter() {
   defaultValue=1
   desc="0==False, 1==True: AO gradient affects Emissive values. Use this for Emissives with depth. Values between 0 && 1 are allowed."
   group=5_Special Effects
   parameterName=Tire_AOTintsEmissive
  }
  n21 = ScalarParameter() {
   desc="0==False, 1==True: Blends AO with Diffuse. Values between 0 && 1 are allowed."
   group=5_Special Effects
   parameterName=Tire_AOTintsDiffuse
  }
  n22 = MaterialFunctionCall(R=@n18, G=@n19, B=@n20, A=@n21) {desc="Vector4 container for Rough, Metallic, AO settings", materialFunction=&AppendMany}
  n23 = TextureSampleParameter2D(UVs=NamedRerouteUsage()) {
   group=1_Input Textures
   parameterName=Tire_Normal
   samplerSource=SSM_Wrap_WorldGroupSettings
   samplerType=SAMPLERTYPE_Normal
   texture=/VehicleCosmetics/Wheels/Esports05_Pro/Textures/Wheel_Mendoza_Tread_Normal_pro
  }
  n24 = LinearInterpolate(A=NamedRerouteUsage(), B=@n23.RGB, Alpha=NamedRerouteUsage())
  n25 = TextureSampleParameter2D(UVs=NamedRerouteUsage()) {
   desc=Diffuse Input Texture - Optional
   group=1_Input Textures
   parameterName=Tire_DiffuseTexture
   samplerSource=SSM_Wrap_WorldGroupSettings
   texture=&T_White
  }
  n26 = ScalarParameter() {
   desc="0==False, 1==True: Set at 1 if Diffuse overrides color, 0 for no Diffuse. Auto-rounds to 0 or 1."
   group=1_Input Textures
   parameterName=Tire_DiffuseTextureEnable
   sliderMax=1
  }
  n27 = VectorParameter() {
   desc="Albedo/Diffuse color control in absence of RGB maps, or to fill in the gaps. Leave white if using Diffuse Texture instead."
   group=4_Surface Defaults
   parameterName=Tire_ColorBase
   defaultValue=<rgba 0.02 0.02 0.02 1>
  }
  n28 = LinearInterpolate(
   A=@n27.RGB
   B=Multiply(
    A=ScalarParameter() {defaultValue=2, group=5_Special Effects, parameterName=Rim Light Brightness}
    B=VectorParameter() {group=5_Special Effects, parameterName=Rim Light Color, defaultValue=<rgba 0.8306 0.2341 1 1>}.RGB
   )
   Alpha=MaterialFunctionCall!(Normal=@n24, Exponent=ScalarParameter() {defaultValue=1, group=5_Special Effects, parameterName=Fresnel}) {materialFunction=/VehicleCosmetics/Wheels/Cola_Holographic/Materials/MF_Fresnel_Classic}.Result
  )
  n29 = TextureSampleParameter2D(UVs=NamedRerouteUsage()) {
   desc=Mask for Tire - Alpha is AO
   group=1_Input Textures
   parameterName=Tire_MaskRGBA
   samplerSource=SSM_Wrap_WorldGroupSettings
   samplerType=SAMPLERTYPE_LinearColor
   texture=/VehicleCosmetics/Wheels/Esports05_Pro/Textures/Wheel_Mendoza_Tread_RGB_pro
   channelNames {a=AO, b=ColorC, g=ColorB, r=ColorA}
  }
  n30 = VectorParameter() {
   desc=Color fill for R value of RGB Mask
   group=3_Tire Properties
   parameterName=Tire_ColorA
   channelNames {a=Unused, b="", g="", r=""}
   defaultValue=<rgba 1 0.1248 0 1>
  }
  n31 = VectorParameter() {
   desc="Emissive, Roughness, Metallic values for ColorA. Emissive has no limit, Roughness && Metallic are clamped between 0 && 1."
   group=3_Tire Properties
   parameterName=Tire_ColorA_PBR
   channelNames {a=Unused, b=Metallic, g=Roughness, r=Emissive}
   defaultValue=<rgba 0 0.7 0 1>
  }
  n32 = VectorParameter() {
   desc=Color fill for G value of RGB Mask
   group=3_Tire Properties
   parameterName=Tire_ColorB
   channelNames {a=Unused, b="", g="", r=""}
   defaultValue=<rgba 0 1 0.5029 1>
  }
  n33 = VectorParameter() {
   desc="Emissive, Roughness, Metallic values for ColorB. Emissive has no limit, Roughness && Metallic are clamped between 0 && 1."
   group=3_Tire Properties
   parameterName=Tire_ColorB_PBR
   channelNames {a=Unused, b=Metallic, g=Roughness, r=Emissive}
   defaultValue=<rgba 0 0.7 0 1>
  }
  n34 = VectorParameter() {
   desc=Color fill for B value of RGB Mask
   group=3_Tire Properties
   parameterName=Tire_ColorC
   channelNames {a=Unused, b="", g="", r=""}
   defaultValue=<rgba 0.5029 0 1 1>
  }
  n35 = VectorParameter() {
   desc="Emissive, Roughness, Metallic values for ColorC. Emissive has no limit, Roughness && Metallic are clamped between 0 && 1."
   group=3_Tire Properties
   parameterName=Tire_ColorC_PBR
   channelNames {a=Unused, b=Metallic, g=Roughness, r=Emissive}
   defaultValue=<rgba 0 0.7 0 1>
  }
  n36 = NamedRerouteDeclaration(None=VertexColor().R) {name=VertColor_R_Out, nodeColor=<rgba 1 0.5 0 1>}
  n37 = GetMaterialAttributes(None=BlendMaterialAttributes(
    A=@n17.Result
    B=MaterialFunctionCall(
     Base_Settings=@n22.RGBA
     Normal_Texture=@n24
     Diffuse_Texture=@n25.RGB
     DiffuseTextureEnable=@n26
     ColorBase=@n28
     MaskRGBA=@n29.RGBA
     ColorA_Color=@n30
     ColorA_PBR_Values=@n31
     ColorB_Color=@n32
     ColorB_PBR_Values=@n33
     ColorC_Color=@n34
     ColorC_PBR_Values=@n35
    ) {materialFunction=&MF_Wheel_Parent_Simple_InputSetup}.Result
    Alpha=@n36
   )) {attributeGetTypes [
    #1
    #2
    #3
    #4
   ]}
  n38 = ScalarParameter() {defaultValue=3, group=5_Special Effects, parameterName=Rim Animated Emissive Brightness}
  n39 = TextureCoordinate()
  n40 = Multiply(
   A=@n38
   B=TextureSample(UVs=Panner(Coordinate=@n39) {speedX=0.5, speedY=0.1}) {texture=/VehicleCosmetics/Wheels/_SharedTextures/FX/Universal_Noise_11_RGB}.B
  )
  n41 = MaterialFunctionCall!(RGBA=@n17.Masks Out) {materialFunction=/VehicleCosmetics/SharedMaterials/MaterialFunctions/MF_RGBA_Splitter}
  n42 = Multiply(A=@n40, B=@n41.B)
  n43 = OneMinus(None=@n36)
  n44 = NamedRerouteUsage()
  Add(A=LinearInterpolate(A=@n37.Emissive Color, B=Multiply(A=@n15, B=@n42), Alpha=Multiply(A=@n43, B=@n41.B)), B=@n44)
  n45 = ConstantBiasScale(None=ComponentMask(None=TextureCoordinate() {vTiling=12}) {g=true}) {bias=-9, scale=32}
  LinearInterpolate(A=@n38, B=@n40, Alpha=OneMinus(None=Clamp(None=@n45)))
  n46 = Add(A=NamedRerouteDeclaration(None=@n37.Base Color) {name=BC, nodeColor=<rgba 1 0 0.5625 1>}, B=@n44)
  n47 = TextureSampleParameter2D() {
   group=5_Special Effects
   parameterName=Opacity_Mask
   samplerType=SAMPLERTYPE_LinearColor
   texture=/VehicleCosmetics/Wheels/Esports05_Pro/New/Textures/T_Esport05Pro_Rim_Alpha_LOD
  }
  n48 = LinearInterpolate(
   A=@n37.Opacity Mask
   B=StaticSwitchParameter(True=@n47.R, False=Constant() {r=1}) {group=5_Special Effects, parameterName=LODs_Mask}
   Alpha=OneMinus(None=NamedRerouteUsage())
  )
  MaterialFunctionCall(In=MaterialFunctionCall(Material=SetMaterialAttributes!(MaterialAttributes=@n37.MaterialAttributes, Base Color=@n46, Emissive Color, Opacity Mask=@n48) {attributeSetTypes [#1, #3, #4]}) {materialFunction=&MF_BaseVehicle_Damage}.Result) {materialFunction=&MF_VehicleDeRez_Transition}
  Multiply(A=@n41.B)
  NamedRerouteDeclaration(None=Panner(Coordinate=TextureCoordinate()) {speedX=-0.1}) {name=Panner UV's, nodeColor=<rgba 0 1 0.1562 1>}
  n49 = TextureSampleParameter2D(UVs=Rotator(Coordinate=TextureCoordinate() {uTiling=16, vTiling=16}, Time=Constant() {r=1}) {speed=1.55}) {
   group=1_Input Textures
   parameterName=Additional Normal
   samplerType=SAMPLERTYPE_Normal
   sortPriority=50
   texture=/VehicleCosmetics/Wheels/_SharedTextures/FX/Universal_Noise_Normal_06
  }
  NamedRerouteDeclaration(None=@n49.RGB) {name=Add Normal, nodeColor=<rgba 0.5 0.5 1 1>}
  NamedRerouteDeclaration(None=Add(A=Add(A=@n29.ColorA, B=@n29.ColorA), B=@n29.ColorC)) {name=Tire Masks, nodeColor=<rgba 0 0.4844 1 1>}
  n50 = ComponentMask(None=Add(A=ReflectionVectorWS(), B=NamedRerouteUsage())) {g=true, r=true}
  n51 = TextureCoordinate() {uTiling=8, vTiling=8}
  n52 = Multiply(A=Time(), B=ScalarParameter() {defaultValue=0.1, group=5_Special Effects, parameterName=Sparkle Speed})
  n53 = Multiply(
   A=Rotator(Coordinate=Panner(Coordinate=@n51, Time=@n52) {speedY=0.01}, Time=@n52) {speed=0.1}
   B=ScalarParameter() {defaultValue=1, group=5_Special Effects, parameterName=Sparkle Tile}
  )
  n54 = Multiply(
   A=TextureSample(UVs=Multiply(A=@n50) {constB=0.0625}) {samplerSource=SSM_Wrap_WorldGroupSettings, texture=/VehicleCosmetics/Wheels/_SharedTextures/FX/Noise_Generic_01_Pack}.B
   B=ConstantBiasScale(None=TextureSample(UVs=@n53) {samplerSource=SSM_Wrap_WorldGroupSettings, texture=&Noise_Generic_01_Pack}.B) {bias=0, scale=1}
  )
  NamedRerouteDeclaration(None=Multiply(A=Multiply(A=@n54) {constB=2048}, B=NamedRerouteUsage())) {name=Sparkles, nodeColor=<rgba 0.2422 0 1 1>}
  NamedRerouteDeclaration(None=@n37.Normal) {name=Normal, nodeColor=<rgba 0.5 0.5 1 1>}
 ]
}
